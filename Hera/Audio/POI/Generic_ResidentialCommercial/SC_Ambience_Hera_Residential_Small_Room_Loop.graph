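# Looping indoor room-tone ambience cue for small residential rooms at Hera POIs, on the indoor ambient sound class with a spot concurrency limit.
# A quality-level node picks between looping quad (first two slots) and stereo (last two slots) versions of the wave.
# Two nested, endlessly looping envelopers of different lengths add a slight pitch drift and gentle volume swell to disguise repetition, under a root Modulator.
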SoundCue /Hera/Audio/POI/Generic_ResidentialCommercial/SC_Ambience_Hera_Residential_Small_Room_Loop
root=Modulator_0
nodes [
 {
  id=Enveloper_6
  children [QualityLevel_0]
  properties {
   bLoop=true
   bLoopIndefinitely=true
   loopEnd=30
   pitchCurve {editorCurveData {keys [
      {arriveTangent=-0.0013, interpMode=RCIM_Cubic, leaveTangent=-0.0013, time=0}
      <at 15 0.98 RCIM_Cubic>
      {arriveTangent=0.0013, interpMode=RCIM_Cubic, leaveTangent=0.0013, time=30}
     ]}}
   volumeCurve {editorCurveData {keys [<at 0 1>]}}
  }
 }
 {
  id=Enveloper_7
  children [Enveloper_6]
  properties {
   bLoop=true
   bLoopIndefinitely=true
   loopEnd=20
   pitchCurve {editorCurveData {keys [{time=0}]}}
   volumeCurve {editorCurveData {keys [
      {arriveTangent=-0.01, interpMode=RCIM_Cubic, time=0, value=1}
      <at 10 0.9 RCIM_Cubic>
      {interpMode=RCIM_Cubic, leaveTangent=0.01, time=20, value=1}
     ]}}
  }
 }
 {id=Modulator_0, children [Enveloper_7]}
 {id=QualityLevel_0, children [WavePlayer_2,WavePlayer_2,WavePlayer_3,WavePlayer_3]}
 {
  id=WavePlayer_2
  properties {bLooping=true, soundWaveAssetPtr=&/Ambience_Hera_Residential_Small_Room_Loop_Quad}
 }
 {
  id=WavePlayer_3
  properties {bLooping=true, soundWaveAssetPtr=&/Ambience_Hera_Residential_Small_Room_Loop_Stereo}
 }
]
properties {
 attenuationSettings=/Game/Attenuations/Fort_Ambient_Athena_Att
 bEnableBusSends=false
 bEnableSubmixSends=false
 maxDistance=2097152
 priority=5
 soundClassObject=&Fort_Ambient_Indoors_Athena
 volumeMultiplier=0.35
 concurrencySet [/Game/Concurrency/AmbientConc_2Max_Spot_Concurrency]
}
